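# Unlit, translucent, two-sided swirl effect material for Niagara mesh particles in the Apprentice emote.
# Two polar-UV noise layers with separate twist amounts scroll over time: one adds a boosted smoothstepped glow, the other samples a curve-atlas color and shapes alpha.
# Particle color tints the glow and scales alpha, with emissive passed through an adjustment function.
# A local X-axis vector, scaled by distance from UV center and the PushSwirl dynamic parameter, forms a push offset.
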Material /BRCosmetics/Animation/Game/MainPlayer/Emotes/Apprentice/FX/M_Apprentice_Emote_Swirl_Synced
graph {
 properties {
  bUsedWithNiagaraMeshParticles=true
  blendMode=BLEND_Translucent
  pixelDepthOffsetMode=PDOM_Legacy
  shadingModel=MSM_Unlit
  twoSided=true
  usedShadingModels=Unlit
 }
 nodes [
  n1 = Distance(A=TextureCoordinate(), B=Constant2Vector() {g=0.5, r=0.5})
  n2 = Multiply(A=Time()) {constB=-0.35}
  n3 = Sine(None=Add(A=@n1, B=Frac(None=@n2))) {period=0.25}
  Add(A=@n3)
  Add(A=@n1, B=@n3)
  ComponentMask(None=ScalarParameter() {parameterName=Param_2}) {g=true, r=true}
  n4 = MaterialFunctionCall!(UV, Twist Center=ScalarParameter() {defaultValue=3, parameterName=TwistAmount_Thin}, Twist Edges) {materialFunction=/Game/Accessories/FORT_Backpacks/Backpack_F_MED_Celestial_Glow/FX/Materials/MF_PolarUV}
  n5 = ScalarParameter() {defaultValue=1, parameterName=Param_5}
  n6 = Time()
  n7 = Constant2Vector() {g=-1, r=1}
  n8 = TextureSample!(
   UVs=Add(A=Multiply(A=@n4.Polar UVs, B=@n5), B=Multiply(A=@n6, B=@n7)) {constB=0}
   Tex
   "DDX(UVs)"=@n4.DDX
   "DDY(UVs)"=@n4.DDX
   Apply View MipBias
  ) {mipValueMode=TMVM_Derivative, samplerType=SAMPLERTYPE_Grayscale, texture=/FortUILibrary/UIKit/Buttons/Textures/LowResBlurredNoise}
  n9 = Multiply(A=SmoothStep(Value=@n8.R) {constMin=0.4}, B=ScalarParameter() {defaultValue=10, parameterName=Param_6}) {constB=400}
  n10 = MaterialFunctionCall(Twist Center=ScalarParameter() {defaultValue=1, parameterName=0}) {materialFunction=&MF_PolarUV}
  n11 = ScalarParameter() {defaultValue=1, parameterName=Param_3}
  n12 = Time()
  n13 = TextureSample(UVs=Add(A=Multiply(A=@n10.Polar UVs, B=@n11), B=Multiply(A=@n12, B=@n7)) {constB=0}, "DDX(UVs)"=@n10.DDX, "DDY(UVs)"=@n10.DDX) {mipValueMode=TMVM_Derivative, samplerType=SAMPLERTYPE_Grayscale, texture=/Game/Global/Textures/Default/Noise/LowResBlurredNoise}
  n14 = Multiply(A=@n13.R) {constB=2}
  n15 = CurveAtlasRowParameter(CurveTime=@n14) {
   atlas=&AthenaCurveAtlas01
   curve=/Game/Athena/LinearColorCurves/Curve_FritterEvent
   defaultValue=212
   parameterName=Param_1
  }
  n16 = ParticleColor()
  MaterialFunctionCall(Emissive=Multiply(A=Add(A=@n9, B=@n15), B=@n16.RGB)) {materialFunction=&MF_FX_EmissiveAdjustments}
  Multiply(A=@n16.A, B=SmoothStep(Value=@n14) {constMin=0.5}) {constA=1}
  n17 = Constant3Vector() {constant=<rgba 1 0 0 0>}
  Multiply(
   A=Multiply(A=Transform(None=@n17) {transformSourceType=TRANSFORMSOURCE_Local}, B=@n1)
   B=DynamicParameter() {defaultValue=<rgba 75 1 1 1>, paramNames [PushSwirl,Param2,Param3,Param4]}.PushSwirl
  )
  ScalarParameter() {defaultValue=1, parameterName=Param_4}
 ]
}
